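# Post-style animation blueprint for the black Slither female character's face accessory, adding physics motion on top of the pose copied from the attached parent mesh.
# Two chained RigidBody nodes simulate relative to the head bone; the second uses a skydive physics asset, weighted by isSkydiving.
# The event graph converts skydiving and crouch-sprint states to floats each update.
# Only isSkydiving feeds the pose; the headbandRt1 bone transform is left unconnected.
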AnimBlueprint /BRCosmetics/Characters/Player/Female/Medium/Bodies/F_MED_Slither/Meshes/Parts/BLK/F_MED_Slither_BLK_FaceAcc_AnimBP
target_skeleton=&/F_MED_Slither_BLK_FaceAcc_Skeleton
graphs [
 AnimGraph {nodes [
   "Transform (Modify) Bone - Bone: headbandRt1"(Alpha=1, Translation="0,0,0", Rotation="0,0,0", Scale="1,1,1") {
    rotationMode=BMM_Additive
    scaleMode=BMM_Additive
    translationMode=BMM_Additive
    alphaBoolBlend {blendOption=Linear}
    alphaScaleBiasClamp {bInterpResult=true, interpSpeedDecreasing=1, interpSpeedIncreasing=1}
    boneToModify {boneName=headbandRt1}
   }
   n1 = RigidBody(ComponentPose=Local To Component(LocalPose=Copy Pose From Mesh() {bUseAttachedParent=true})) {
    bClampLinearTranslationLimitToRefPose=true
    bOverrideWorldGravity=true
    overridePhysicsAsset=&/F_MED_Slither_BLK_FaceAcc_Physics
    simulationSpace=BaseBoneSpace
    alphaBoolBlend {blendOption=Linear}
    baseBoneRef {boneName=head}
    componentAppliedLinearAccClamp=<3000 3000 3000>
    componentLinearAccScale=<0.4 0.3 0.3>
    componentLinearVelScale=<0.4 0.3 0.3>
    overrideWorldGravity {z=-980}
   }
   n2 = RigidBody(ComponentPose=@n1, Alpha=Get isSkydiving() -> (isSkydiving:double)) {
    bClampLinearTranslationLimitToRefPose=true
    bOverrideWorldGravity=true
    overridePhysicsAsset=&/F_MED_Slither_BLK_FaceAcc_PhysicsSkydive
    simulationSpace=BaseBoneSpace
    alphaBoolBlend {blendOption=Linear}
    baseBoneRef {boneName=head}
    componentAppliedLinearAccClamp=<3000 3000 3000>
    componentLinearAccScale=<0.4 0.3 0.3>
    componentLinearVelScale=<0.4 0.3 0.3>
    overrideWorldGravity {z=-980}
   }
   Output Pose(Result=Component To Local(ComponentPose=@n2)) {layerGroup=None}
  ]}
 EventGraph {nodes [
   Event BlueprintUpdateAnimation()
   > Set crouchWalking(crouchWalking:double="To Float (Boolean)"(InBool=Get bIsCrouchSprinting())) -> (Output_Get:double)
   n1 = Get bIsSkyDiving()
   > Set notSkydiving(notSkydiving:double="To Float (Boolean)"(InBool=NOT Boolean(A=@n1))) -> (Output_Get:double)
   > n2 = Branch(Condition=Get bIsSkyDiving())
   > Set isSkydiving(isSkydiving:double=1) -> (Output_Get:double)
   Set isSkydiving(execute=@n2.else, isSkydiving=0)
  ]}
]
variables [
 crouchWalking {type=number}
 isSkydiving {type=number}
 notSkydiving {type=number}
]
